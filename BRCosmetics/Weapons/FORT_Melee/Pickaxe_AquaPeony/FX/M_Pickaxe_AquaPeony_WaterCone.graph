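# Translucent unlit water-cone material for Niagara mesh particles in the AquaPeony pickaxe effects.
# Panning cloud noise and cell-noise caustics (dropped on mobile) build wave masks that blend light water, dark water and foam colors, tinted by particle color.
# Dynamic parameters drive vertical offset, opacity and normal-based vertex displacement.
# Opacity combines particle-alpha erosion with a clip range, inverted Fresnel edge softening and a UV gradient fade.
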Material /BRCosmetics/Weapons/FORT_Melee/Pickaxe_AquaPeony/FX/M_Pickaxe_AquaPeony_WaterCone
graph {
 properties {
  bAutomaticallySetUsageInEditor=false
  bUsedWithNiagaraMeshParticles=true
  blendMode=BLEND_Translucent
  shadingModel=MSM_Unlit
  usedShadingModels=Unlit
 }
 nodes [
  n1 = VectorParameter() {parameterName=FoamColor, defaultValue=<rgba 0.8197 0.8988 1 1>}
  n2 = NamedRerouteUsage()
  n3 = Saturate(None=Multiply(A=Power(Base=@n2) {constExponent=3}) {constB=2})
  n4 = Power(Base=NamedRerouteUsage()) {constExponent=0.8}
  n5 = LinearInterpolate(
   A=@n1.RGB
   B=LinearInterpolate(
    A=VectorParameter() {parameterName=WaterLight, defaultValue=<rgba 0.4 0.69 1 1>}.RGB
    B=VectorParameter() {parameterName=WaterDark, defaultValue=<rgba 0.135 0.1875 0.45 1>}.RGB
    Alpha=NamedRerouteUsage()
   )
   Alpha=Multiply(A=OneMinus(None=@n3), B=SmoothStep(Value=@n4) {constMax=0.06})
  )
  n6 = NamedRerouteUsage()
  MaterialFunctionCall(Emissive=Multiply(A=LinearInterpolate(A=@n5, B=@n1.RGB, Alpha=@n6), B=ParticleColor().RGB)) {materialFunction=&MF_FX_EmissiveAdjustments}
  n7 = MaterialFunctionCall(Float2=TextureCoordinate()) {materialFunction=&BreakOutFloat2Components}
  n8 = Power(Base=NamedRerouteUsage()) {constExponent=0.4}
  n9 = Constant3Vector() {constant=<rgba 1 1 1 0>}
  n10 = Multiply(A=VertexNormalWS(), B=@n9)
  n11 = DynamicParameter() {defaultValue=<rgba 0 1 1 1>, paramNames [OffsetY,Opacity,WPOIntensity,Param4]}
  Multiply(A=Saturate(None=Multiply(A=@n7.G) {constB=3}), B=Multiply(A=Multiply(A=@n8, B=@n10), B=Multiply(A=@n11.WPOIntensity) {constB=8}) {constB=6})
  n12 = NamedRerouteUsage()
  n13 = Power(Base=Add(A=Multiply(A=@n12) {constB=0.4}, B=NamedRerouteUsage())) {constExponent=11}
  NamedRerouteDeclaration(None=Saturate(None=Multiply(A=@n13) {constB=3})) {name=OriginMask, nodeColor=<rgba 1 0.2336 0.0752 1>}
  n14 = MaterialFunctionCall(Float2=TextureCoordinate()) {materialFunction=&BreakOutFloat2Components}
  n15 = AppendVector(
   A=@n14.R
   B=Add(A=OneMinus(None=@n14.G), B=DynamicParameter() {defaultValue=<rgba 0 1 1 1>, paramNames [OffsetY,Opacity,WPOIntensity,Param4]}.OffsetY) {constB=-0.1}
  )
  n16 = Multiply(A=@n15, B=Constant2Vector() {g=0.33, r=2})
  NamedRerouteDeclaration(None=TextureSampleParameter2D(UVs=Panner(Coordinate=@n16) {speedY=0.3}) {parameterName=WaveNoise, samplerType=SAMPLERTYPE_LinearColor, texture=/Game/Global/Textures/Default/Noise/T_CloudNoise_01}.R) {name=WaveNoise, nodeColor=<rgba 0.1607 0.7198 1 1>}
  NamedRerouteDeclaration(None=MaterialFunctionCall(Float2=@n15) {materialFunction=&BreakOutFloat2Components}.G) {name=TopDownMask, nodeColor=<rgba 1 0 0.4453 1>}
  n17 = Multiply(A=TextureCoordinate(), B=Constant2Vector() {g=0.33, r=1})
  n18 = Power(Base=TextureSampleParameter2D(UVs=Panner(Coordinate=@n17) {speedY=-0.66}) {parameterName=CausticNoise, texture=/Game/Global/Textures/Default/Noise/T_CellNoise_01}.R) {constExponent=7}
  n19 = OneMinus(None=NamedRerouteUsage())
  n20 = Multiply(A=Saturate(None=Multiply(A=@n19) {constB=4}), B=NamedRerouteUsage())
  n21 = Max(
   A=ShadingPathSwitch(Default=@n18, Deferred=@n18, Mobile=Constant()) {desc=Migrated from FeatureLevelSwitch}
   B=LinearInterpolate(A=NamedRerouteUsage(), Alpha=@n20)
  )
  n22 = NamedRerouteUsage()
  NamedRerouteDeclaration(None=Multiply(A=Multiply(A=@n21, B=@n20), B=Power(Base=@n22) {constExponent=3})) {name=WaveMask, nodeColor=<rgba 0 1 0.2734 1>}
  n23 = MaterialFunctionCall(Float2=TextureCoordinate()) {materialFunction=&BreakOutFloat2Components}
  n24 = Saturate(None=Multiply(A=OneMinus(None=@n23.G)) {constB=3})
  n25 = Fresnel() {baseReflectFraction=0, exponent=1}
  n26 = NamedRerouteUsage()
  n27 = NamedRerouteUsage()
  n28 = Power(Base=NamedRerouteUsage()) {constExponent=5}
  n29 = Subtract(A=Saturate(None=Add(A=@n27, B=@n28)), B=LinearInterpolate(Alpha=ParticleColor().A) {constA=1, constB=-1})
  n30 = ScalarParameter() {defaultValue=0.05, group=Opacity, parameterName=ClipRange}
  n31 = ScalarParameter() {defaultValue=0.9, group=Opacity, parameterName=ClipAmnt}
  n32 = Subtract(
   A=Multiply(A=SmoothStep(Value=@n26) {constMax=0.11, constMin=0.01}, B=Saturate(None=@n29))
   B=LinearInterpolate(A=Multiply(A=@n30) {constB=-1}, Alpha=OneMinus(None=@n31) {desc=Inverse Alpha})
  )
  n33 = Multiply(A=Power(Base=OneMinus(None=@n25)) {constExponent=1.5}, B=Saturate(None=Divide(A=@n32, B=@n30)))
  Saturate(None=Multiply(A=DynamicParameter() {defaultValue=<rgba 0 1 1 1>, paramNames [OffsetY,Opacity,WPOIntensity,Param4]}.Opacity, B=Multiply(A=@n24, B=@n33)))
  TransformPosition(None=@n9) {transformSourceType=TRANSFORMPOSSOURCE_World}
 ]
}
